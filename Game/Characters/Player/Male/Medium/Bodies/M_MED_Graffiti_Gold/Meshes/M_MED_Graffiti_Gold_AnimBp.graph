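# Post-process style AnimBlueprint for the M_MED_Graffiti_Gold player body that keeps its collar clear during head, arm and body movement.
# It copies the parent mesh pose through a Control Rig, layers additive collar-bone correctives, simulates the zipper with AnimDynamics, and shrinks the head for custom-character-part playlists.
# Cached poses branch into aiming, skydiving and floating adjustments.
# The event graph derives blend alphas from bone deltas, lateral velocity and movement states.
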AnimBlueprint /Game/Characters/Player/Male/Medium/Bodies/M_MED_Graffiti_Gold/Meshes/M_MED_Graffiti_Gold_AnimBp
target_skeleton=&/M_MED_Graffiti_Gold_Skeleton
graphs [
 AnimGraph {nodes [
   n1 = Get Head rotation back skyfloating() -> (Head rotation back skyfloating:double)
   n2 = "Transform (Modify) Bone - Bone: dyn_collar_bk_01"(Alpha=@n1, Translation="0,0,6", Rotation="90,0,0", Scale="1,1,1") {
    rotationMode=BMM_Additive
    rotationSpace=BCS_BoneSpace
    scaleSpace=BCS_BoneSpace
    translationMode=BMM_Additive
    translationSpace=BCS_BoneSpace
    alphaBoolBlend {blendOption=Linear}
    boneToModify {boneName=dyn_collar_bk_01}
    rotation {pitch=90}
    translation {z=6}
   }
   n3 = "Transform (Modify) Bone - Bone: dyn_collar_back_r_01"(ComponentPose=@n2, Alpha=@n1, Translation="6,0,0", Rotation="0,0,0", Scale="1,1,1") {
    rotationMode=BMM_Additive
    rotationSpace=BCS_BoneSpace
    scaleSpace=BCS_BoneSpace
    translationMode=BMM_Additive
    translationSpace=BCS_BoneSpace
    alphaBoolBlend {blendOption=Linear}
    boneToModify {boneName=dyn_collar_back_r_01}
    translation {x=6}
   }
   "Transform (Modify) Bone - Bone: dyn_collar_back_l_01"(ComponentPose=@n3, Alpha=@n1, Translation="6,0,0", Rotation="0,0,0", Scale="1,1,1") {
    rotationMode=BMM_Additive
    rotationSpace=BCS_BoneSpace
    scaleSpace=BCS_BoneSpace
    translationMode=BMM_Additive
    translationSpace=BCS_BoneSpace
    alphaBoolBlend {blendOption=Linear}
    boneToModify {boneName=dyn_collar_back_l_01}
    translation {x=6}
   }
   n4 = Get LeftMovementAlphaSkydiving() -> (LeftMovementAlphaSkydiving:double)
   n5 = "Transform (Modify) Bone - Bone: dyn_collar_back_l_01"(
    ComponentPose=Local To Component(LocalPose=Use cached pose 'Skydiving Pose'())
    Alpha=@n4
    Translation="0,0,0"
    Rotation="0,0,0"
    Scale="1,1,1"
   ) {
    rotationMode=BMM_Additive
    rotationSpace=BCS_BoneSpace
    translationMode=BMM_Additive
    translationSpace=BCS_BoneSpace
    alphaBoolBlend {blendOption=Linear}
    alphaScaleBiasClamp {interpSpeedDecreasing=0.5, interpSpeedIncreasing=0.5}
    boneToModify {boneName=dyn_collar_back_l_01}
   }
   n6 = "Transform (Modify) Bone - Bone: dyn_collar_side_l_01"(ComponentPose=@n5, Alpha=@n4, Translation="0,0,0", Rotation="0,0,0", Scale="1,1,1") {
    rotationMode=BMM_Additive
    rotationSpace=BCS_BoneSpace
    translationMode=BMM_Additive
    translationSpace=BCS_BoneSpace
    alphaBoolBlend {blendOption=Linear}
    alphaScaleBiasClamp {interpSpeedDecreasing=0.5, interpSpeedIncreasing=0.5}
    boneToModify {boneName=dyn_collar_side_l_01}
   }
   n7 = Get RightMovementAlphaSkydiving() -> (RightMovementAlphaSkydiving:double)
   n8 = "Transform (Modify) Bone - Bone: dyn_collar_back_l_01"(ComponentPose=@n6, Alpha=@n7, Translation="0,0,0", Rotation="-5,0,0", Scale="1,1,1") {
    rotationMode=BMM_Additive
    rotationSpace=BCS_BoneSpace
    translationMode=BMM_Additive
    translationSpace=BCS_BoneSpace
    alphaBoolBlend {blendOption=Linear}
    alphaScaleBiasClamp {interpSpeedDecreasing=0.5, interpSpeedIncreasing=0.5}
    boneToModify {boneName=dyn_collar_back_l_01}
    rotation {pitch=-5}
   }
   n9 = "Transform (Modify) Bone - Bone: dyn_collar_side_r_01"(ComponentPose=@n8, Alpha=@n7, Translation="0,0,0", Rotation="0,0,0", Scale="1,1,1") {
    rotationMode=BMM_Additive
    rotationSpace=BCS_BoneSpace
    translationMode=BMM_Additive
    translationSpace=BCS_BoneSpace
    alphaBoolBlend {blendOption=Linear}
    alphaScaleBiasClamp {interpSpeedDecreasing=0.5, interpSpeedIncreasing=0.5}
    boneToModify {boneName=dyn_collar_side_r_01}
   }
   n10 = Get Left arm skydiving() -> (Left arm skydiving:double)
   n11 = "Transform (Modify) Bone - Bone: dyn_collar_side_l_01"(ComponentPose=@n9, Alpha=@n10, Translation="0,0,0", Rotation="-60,0,0", Scale="1,1,1") {
    rotationMode=BMM_Additive
    rotationSpace=BCS_BoneSpace
    translationMode=BMM_Additive
    translationSpace=BCS_BoneSpace
    alphaBoolBlend {blendOption=Linear}
    alphaScaleBiasClamp {interpSpeedDecreasing=0.5, interpSpeedIncreasing=0.5}
    boneToModify {boneName=dyn_collar_side_l_01}
    rotation {pitch=-60}
   }
   n12 = "Transform (Modify) Bone - Bone: dyn_collar_back_l_01"(ComponentPose=@n11, Alpha=@n10, Translation="0,0,0", Rotation="10,0,0", Scale="1,1,1") {
    rotationMode=BMM_Additive
    rotationSpace=BCS_BoneSpace
    translationMode=BMM_Additive
    translationSpace=BCS_BoneSpace
    alphaBoolBlend {blendOption=Linear}
    alphaScaleBiasClamp {interpSpeedDecreasing=0.5, interpSpeedIncreasing=0.5}
    boneToModify {boneName=dyn_collar_back_l_01}
    rotation {pitch=10}
   }
   n13 = Get Right arm skydiving() -> (Right arm skydiving:double)
   n14 = "Transform (Modify) Bone - Bone: dyn_collar_side_r_01"(ComponentPose=@n12, Alpha=@n13, Translation="0,0,0", Rotation="-60,0,0", Scale="1,1,1") {
    rotationMode=BMM_Additive
    rotationSpace=BCS_BoneSpace
    translationMode=BMM_Additive
    translationSpace=BCS_BoneSpace
    alphaBoolBlend {blendOption=Linear}
    alphaScaleBiasClamp {interpSpeedDecreasing=0.5, interpSpeedIncreasing=0.5}
    boneToModify {boneName=dyn_collar_side_r_01}
    rotation {pitch=-60}
   }
   n15 = "Transform (Modify) Bone - Bone: dyn_collar_back_r_01"(ComponentPose=@n14, Alpha=@n13, Translation="0,0,0", Rotation="10,0,0", Scale="1,1,1") {
    rotationMode=BMM_Additive
    rotationSpace=BCS_BoneSpace
    translationMode=BMM_Additive
    translationSpace=BCS_BoneSpace
    alphaBoolBlend {blendOption=Linear}
    alphaScaleBiasClamp {interpSpeedDecreasing=0.5, interpSpeedIncreasing=0.5}
    boneToModify {boneName=dyn_collar_back_r_01}
    rotation {pitch=10}
   }
   Output Pose(Result=Blend Poses by bool(BlendPose_0=Component To Local(ComponentPose=@n15), BlendPose_1=Use cached pose 'Skydiving Pose'(), bActiveValue=Get bIsSkydiveFloating()) {blendType=Linear}) {layerGroup=None}
   n16 = Control Rig(Source=Copy Pose From Mesh(SourceMeshComponent=Get MeshToCopy() -> (MeshToCopy:Skeletal Mesh Component Object Reference)) {bUseAttachedParent=true}) {
    alphaBoolBlend {blendOption=Linear}
    controlRigAssetReference {blueprintRigClass=&/M_MED_Graffiti_Gold_CtrlRig.M_MED_Graffiti_Gold_CtrlRig_C}
   }
   n17 = Get head_rot_back() -> (head_rot_back:double)
   n18 = "Transform (Modify) Bone - Bone: dyn_collar_bk_01"(ComponentPose=Local To Component(LocalPose=@n16), Alpha=@n17, Translation="0,2,2", Rotation="90,0,0", Scale="1,1,1") {
    rotationMode=BMM_Additive
    rotationSpace=BCS_BoneSpace
    scaleSpace=BCS_BoneSpace
    translationMode=BMM_Additive
    translationSpace=BCS_BoneSpace
    alphaBoolBlend {blendOption=Linear}
    boneToModify {boneName=dyn_collar_bk_01}
    rotation {pitch=90}
    translation {y=2, z=2}
   }
   n19 = "Transform (Modify) Bone - Bone: dyn_collar_back_r_01"(ComponentPose=@n18, Alpha=@n17, Translation="0,0,0", Rotation="0,0,0", Scale="1,1,1") {
    rotationMode=BMM_Additive
    rotationSpace=BCS_BoneSpace
    scaleSpace=BCS_BoneSpace
    translationMode=BMM_Additive
    translationSpace=BCS_BoneSpace
    alphaBoolBlend {blendOption=Linear}
    boneToModify {boneName=dyn_collar_back_r_01}
   }
   n20 = "Transform (Modify) Bone - Bone: dyn_collar_back_l_01"(ComponentPose=@n19, Alpha=@n17, Translation="0,0,0", Rotation="0,0,0", Scale="1,1,1") {
    rotationMode=BMM_Additive
    rotationSpace=BCS_BoneSpace
    scaleSpace=BCS_BoneSpace
    translationMode=BMM_Additive
    translationSpace=BCS_BoneSpace
    alphaBoolBlend {blendOption=Linear}
    boneToModify {boneName=dyn_collar_back_l_01}
   }
   n21 = Get head_rot_back_not_targeting() -> (head_rot_back_not_targeting:double)
   n22 = "Transform (Modify) Bone - Bone: dyn_collar_back_r_01"!(ComponentPose=@n20, Alpha=@n21, Translation="3,0,0", Rotation="10,0,0") -> (Pose) {
    rotationMode=BMM_Additive
    rotationSpace=BCS_BoneSpace
    scaleSpace=BCS_BoneSpace
    translationMode=BMM_Additive
    translationSpace=BCS_BoneSpace
    alphaBoolBlend {blendOption=Linear}
    boneToModify {boneName=dyn_collar_back_r_01}
    rotation {pitch=10}
    scale=<2 2 2>
    translation {x=3}
   }
   n23 = "Transform (Modify) Bone - Bone: dyn_collar_back_l_01"!(ComponentPose=@n22, Alpha=@n21, Translation="2,0,2", Rotation="30,0,0") -> (Pose) {
    rotationMode=BMM_Additive
    rotationSpace=BCS_BoneSpace
    scaleSpace=BCS_BoneSpace
    translationMode=BMM_Additive
    translationSpace=BCS_BoneSpace
    alphaBoolBlend {blendOption=Linear}
    boneToModify {boneName=dyn_collar_back_l_01}
    rotation {pitch=30}
    scale=<2 2 2>
    translation {x=2, z=2}
   }
   n24 = "Transform (Modify) Bone - Bone: dyn_collar_side_l_01"!(ComponentPose=@n23, Alpha=@n21, Translation="0,0,2", Rotation="45,0,0") -> (Pose) {
    rotationMode=BMM_Additive
    rotationSpace=BCS_BoneSpace
    scaleSpace=BCS_BoneSpace
    translationMode=BMM_Additive
    translationSpace=BCS_BoneSpace
    alphaBoolBlend {blendOption=Linear}
    boneToModify {boneName=dyn_collar_side_l_01}
    rotation {pitch=45}
    scale=<2 2 2>
    translation {z=2}
   }
   n25 = "Transform (Modify) Bone - Bone: dyn_collar_side_l_01"(ComponentPose=@n24, Alpha=Get Neck Side Left() -> (Neck Side Left:float), Translation="6,0,0", Rotation="-35,0,0") {
    rotationMode=BMM_Additive
    rotationSpace=BCS_BoneSpace
    scaleSpace=BCS_BoneSpace
    translationMode=BMM_Additive
    translationSpace=BCS_BoneSpace
    alphaBoolBlend {blendOption=Linear}
    boneToModify {boneName=dyn_collar_side_l_01}
    rotation {pitch=-35}
    scale=<2 2 2>
    translation {x=6}
   }
   n26 = "Transform (Modify) Bone - Bone: dyn_collar_front_l_01"!(ComponentPose=@n25, Alpha=Get Neck Side Left(), Translation="0,0,0", Rotation="-20,10,0") -> (Pose) {
    rotationMode=BMM_Additive
    rotationSpace=BCS_BoneSpace
    scaleSpace=BCS_BoneSpace
    translationMode=BMM_Additive
    translationSpace=BCS_BoneSpace
    alphaBoolBlend {blendOption=Linear}
    boneToModify {boneName=dyn_collar_front_l_01}
    rotation {pitch=-20, yaw=10}
    scale=<2 2 2>
   }
   n27 = "Transform (Modify) Bone - Bone: dyn_collar_back_l_01"(ComponentPose=@n26, Alpha=Get Neck Side Left(), Translation="3,0,0", Rotation="-30,10,0") {
    rotationMode=BMM_Additive
    rotationSpace=BCS_BoneSpace
    scaleSpace=BCS_BoneSpace
    translationMode=BMM_Additive
    translationSpace=BCS_BoneSpace
    alphaBoolBlend {blendOption=Linear}
    boneToModify {boneName=dyn_collar_back_l_01}
    rotation {pitch=-30, yaw=10}
    scale=<2 2 2>
    translation {x=3}
   }
   n28 = "Transform (Modify) Bone - Bone: dyn_collar_side_r_01"!(ComponentPose=@n27, Alpha=Get Neck Side Right() -> (Neck Side Right:float), Translation="3,0,-1", Rotation="-56,0,0") -> (Pose) {
    rotationMode=BMM_Additive
    rotationSpace=BCS_BoneSpace
    scaleSpace=BCS_BoneSpace
    translationMode=BMM_Additive
    translationSpace=BCS_BoneSpace
    alphaBoolBlend {blendOption=Linear}
    boneToModify {boneName=dyn_collar_side_r_01}
    rotation {pitch=-56}
    scale=<2 2 2>
    translation {x=3, z=-1}
   }
   n29 = "Transform (Modify) Bone - Bone: dyn_collar_front_r_01"!(ComponentPose=@n28, Alpha=Get Neck Side Right(), Translation="0,0,0", Rotation="-30,0,0") -> (Pose) {
    rotationMode=BMM_Additive
    rotationSpace=BCS_BoneSpace
    scaleSpace=BCS_BoneSpace
    translationMode=BMM_Additive
    translationSpace=BCS_BoneSpace
    alphaBoolBlend {blendOption=Linear}
    boneToModify {boneName=dyn_collar_front_r_01}
    rotation {pitch=-30}
    scale=<2 2 2>
   }
   n30 = "Transform (Modify) Bone - Bone: dyn_collar_back_r_01"(ComponentPose=@n29, Alpha=Get Neck Side Right(), Translation="3,0,0", Rotation="-30,10,0") {
    rotationMode=BMM_Additive
    rotationSpace=BCS_BoneSpace
    scaleSpace=BCS_BoneSpace
    translationMode=BMM_Additive
    translationSpace=BCS_BoneSpace
    alphaBoolBlend {blendOption=Linear}
    boneToModify {boneName=dyn_collar_back_r_01}
    rotation {pitch=-30, yaw=10}
    scale=<2 2 2>
    translation {x=3}
   }
   n31 = Get Alpha Output Idle() -> (Alpha Output Idle:float)
   n32 = "Transform (Modify) Bone - Bone: dyn_collar_back_r_01"(ComponentPose=@n30, Alpha=@n31, Translation="4,0,-3", Rotation="-35,0,0") {
    rotationMode=BMM_Additive
    rotationSpace=BCS_BoneSpace
    scaleSpace=BCS_BoneSpace
    translationMode=BMM_Additive
    translationSpace=BCS_BoneSpace
    alphaBoolBlend {blendOption=Linear}
    boneToModify {boneName=dyn_collar_back_r_01}
    rotation {pitch=-35}
    scale=<2 2 2>
    translation {x=4, z=-3}
   }
   n33 = "Transform (Modify) Bone - Bone: dyn_collar_back_l_01"(ComponentPose=@n32, Alpha=@n31, Translation="-3,0,-3", Rotation="-25,0,20") {
    rotationMode=BMM_Additive
    rotationSpace=BCS_BoneSpace
    scaleSpace=BCS_BoneSpace
    translationMode=BMM_Additive
    translationSpace=BCS_BoneSpace
    alphaBoolBlend {blendOption=Linear}
    boneToModify {boneName=dyn_collar_back_l_01}
    rotation {pitch=-25, roll=20}
    scale=<2 2 2>
    translation {x=-3, z=-3}
   }
   n34 = "Transform (Modify) Bone - Bone: dyn_collar_bk_01"!(ComponentPose=@n33, Alpha=@n31, Translation="5,0,4", Rotation="0,0,0") -> (Pose) {
    rotationMode=BMM_Additive
    rotationSpace=BCS_BoneSpace
    scaleSpace=BCS_BoneSpace
    translationMode=BMM_Additive
    translationSpace=BCS_BoneSpace
    alphaBoolBlend {blendOption=Linear}
    boneToModify {boneName=dyn_collar_bk_01}
    scale=<2 2 2>
    translation {x=5, z=4}
   }
   n35 = Get Alpha Left Arm() -> (Alpha Left Arm:float)
   n36 = "Transform (Modify) Bone - Bone: dyn_collar_back_l_01"(ComponentPose=@n34, Alpha=@n35, Translation="-3,0,0", Rotation="-25,0,0") {
    rotationMode=BMM_Additive
    rotationSpace=BCS_BoneSpace
    scaleSpace=BCS_BoneSpace
    translationMode=BMM_Additive
    translationSpace=BCS_BoneSpace
    alphaBoolBlend {blendOption=Linear}
    boneToModify {boneName=dyn_collar_back_l_01}
    rotation {pitch=-25}
    scale=<2 2 2>
    translation {x=-3}
   }
   n37 = "Transform (Modify) Bone - Bone: dyn_collar_bk_01"(ComponentPose=@n36, Alpha=@n35, Translation="0,0,3", Rotation="0,0,0") {
    rotationMode=BMM_Additive
    rotationSpace=BCS_BoneSpace
    scaleSpace=BCS_BoneSpace
    translationMode=BMM_Additive
    translationSpace=BCS_BoneSpace
    alphaBoolBlend {blendOption=Linear}
    boneToModify {boneName=dyn_collar_bk_01}
    scale=<2 2 2>
    translation {z=3}
   }
   n38 = "Constraint - dyn_zipper (PPPPPP)"!(
    ComponentPose=@n37
    Alpha=1
    ConstraintWeights_0=0.02
    ConstraintWeights_1:float=0.0352
    ConstraintWeights_2:float=0.0015
    ConstraintWeights_3:float=0.2435
    ConstraintWeights_4:float=0.0235
    ConstraintWeights_5:float=0.6764
   ) -> (Pose) {
    alphaBoolBlend {blendOption=Linear}
    boneToModify {boneName=dyn_zipper}
    constraintSetup [
     {targetBone {boneName=clavicle_l}}
     {targetBone {boneName=thigh_twist_01_r}}
     {targetBone {boneName=spine_01}}
     {targetBone {boneName=thigh_twist_01_l}}
     {targetBone {boneName=thigh_l}}
     {targetBone {boneName=pelvis}}
    ]
    constraintWeights [1,1,1,1,1,1]
   }
   n39 = Anim Dynamics - Bone: dyn_zipper(ComponentPose=@n38, Alpha=1) {
    angularDampingOverride=0.96
    bGravityOverrideInSimSpace=true
    bOverrideAngularDamping=true
    bOverrideLinearDamping=true
    gravityScale=0.25
    linearDampingOverride=0.96
    simulationSpace=Actor
    alphaBoolBlend {blendOption=Linear}
    boundBone {boneName=dyn_zipper}
    componentAppliedLinearAccClamp=<700 700 700>
    componentLinearAccScale=<0.15 0.15 0.15>
    componentLinearVelScale=<0.15 0.15 0.15>
    physicsBodyDefinitions [{
      sphereCollisionRadius=0
      boundBone {boneName=dyn_zipper}
      boxExtents=<8 2 4>
      constraintSetup {angularLimitsMax=<0 0 50>, angularLimitsMin=<0 0 -20>, angularTarget=<0 0 0>}
      localJointOffset=<2 0 0>
     }]
   }
   Save cached pose 'CollarRotCache'(Pose=Component To Local(ComponentPose=@n39)) {cacheName=CollarRotCache}
   n40 = "Transform (Modify) Bone - Bone: head"!(ComponentPose=Local To Component(LocalPose=Use cached pose 'CollarRotCache'()), Alpha=1) -> (Pose) {scaleMode=BMM_Replace, alphaBoolBlend {blendOption=Linear}, boneToModify {boneName=head}, scale=<0.01 0.01 0.01>}
   Save cached pose 'OutPose'(Pose=Blend Poses by bool(
     BlendPose_0=Component To Local(ComponentPose=@n40)
     BlendPose_1=Use cached pose 'CollarRotCache'()
     BlendTime_0=0
     BlendTime_1=0
     bActiveValue=Get GetPlayerListUsesCustomCharacterParts() -> (GetPlayerListUsesCustomCharacterParts:bool)
    ) {blendType=Linear}) {cacheName=OutPose}
   n41 = Get ads_rotate_back() -> (ads_rotate_back:double)
   n42 = "Transform (Modify) Bone - Bone: dyn_collar_back_r_01"(ComponentPose=Local To Component(LocalPose=Use cached pose 'OutPose'()), Alpha=@n41, Translation="6,0,0", Rotation="10,0,0") {
    rotationMode=BMM_Additive
    rotationSpace=BCS_BoneSpace
    scaleSpace=BCS_BoneSpace
    translationMode=BMM_Additive
    translationSpace=BCS_BoneSpace
    alphaBoolBlend {blendOption=Linear}
    boneToModify {boneName=dyn_collar_back_r_01}
    rotation {pitch=10}
    scale=<2 2 2>
    translation {x=6}
   }
   n43 = "Transform (Modify) Bone - Bone: dyn_collar_back_l_01"(ComponentPose=@n42, Alpha=@n41, Translation="-3,-2,0", Rotation="-25,0,20") {
    rotationMode=BMM_Additive
    rotationSpace=BCS_BoneSpace
    scaleSpace=BCS_BoneSpace
    translationMode=BMM_Additive
    translationSpace=BCS_BoneSpace
    alphaBoolBlend {blendOption=Linear}
    boneToModify {boneName=dyn_collar_back_l_01}
    rotation {pitch=-25, roll=20}
    scale=<2 2 2>
    translation=<-3 -2>
   }
   n44 = "Transform (Modify) Bone - Bone: dyn_collar_bk_01"(ComponentPose=@n43, Alpha=@n41, Translation="0,0,3", Rotation="0,0,0") {
    rotationMode=BMM_Additive
    rotationSpace=BCS_BoneSpace
    scaleSpace=BCS_BoneSpace
    translationMode=BMM_Additive
    translationSpace=BCS_BoneSpace
    alphaBoolBlend {blendOption=Linear}
    boneToModify {boneName=dyn_collar_bk_01}
    scale=<2 2 2>
    translation {z=3}
   }
   n45 = "Transform (Modify) Bone - Bone: dyn_collar_side_l_01"(ComponentPose=@n44, Alpha=@n41, Translation="0,0,3", Rotation="30,0,-20") {
    rotationMode=BMM_Additive
    rotationSpace=BCS_BoneSpace
    scaleSpace=BCS_BoneSpace
    translationMode=BMM_Additive
    translationSpace=BCS_BoneSpace
    alphaBoolBlend {blendOption=Linear}
    boneToModify {boneName=dyn_collar_side_l_01}
    rotation {pitch=30, roll=-20}
    scale=<2 2 2>
    translation {z=3}
   }
   n46 = Component To Local(ComponentPose=@n45)
   n47 = Use cached pose 'OutPose'()
   n48 = Get bIsTargeting()
   n49 = "Get  jump alpha"() -> (" jump alpha":double)
   n50 = "Transform (Modify) Bone - Bone: dyn_collar_back_l_01"(
    ComponentPose=Local To Component(LocalPose=Blend Poses by bool(BlendPose_0=@n46, BlendPose_1=@n47, BlendTime_0=0, BlendTime_1=0, bActiveValue=@n48) {blendType=Linear})
    Alpha=@n49
    Translation="-2,0,0"
    Rotation="5,0,0"
   ) {
    rotationMode=BMM_Additive
    rotationSpace=BCS_BoneSpace
    scaleSpace=BCS_BoneSpace
    translationMode=BMM_Additive
    translationSpace=BCS_BoneSpace
    alphaBoolBlend {blendOption=Linear}
    alphaScaleBiasClamp {bInterpResult=true, interpSpeedIncreasing=2}
    boneToModify {boneName=dyn_collar_back_l_01}
    rotation {pitch=5}
    scale=<2 2 2>
    translation {x=-2}
   }
   n51 = "Transform (Modify) Bone - Bone: dyn_collar_side_l_01"(ComponentPose=@n50, Alpha=@n49, Translation="-1,0,0", Rotation="10,0,0") {
    rotationMode=BMM_Additive
    rotationSpace=BCS_BoneSpace
    scaleSpace=BCS_BoneSpace
    translationMode=BMM_Additive
    translationSpace=BCS_BoneSpace
    alphaBoolBlend {blendOption=Linear}
    alphaScaleBiasClamp {bInterpResult=true, interpSpeedIncreasing=2}
    boneToModify {boneName=dyn_collar_side_l_01}
    rotation {pitch=10}
    scale=<2 2 2>
    translation {x=-1}
   }
   n52 = Get Head rotation back skydiving() -> (Head rotation back skydiving:double)
   n53 = "Transform (Modify) Bone - Bone: dyn_collar_bk_01"!(ComponentPose=@n51, Alpha=@n52, Translation="0,0,5", Rotation="90,0,0", Scale:Vector="1,1,1") -> (Pose) {
    rotationMode=BMM_Additive
    rotationSpace=BCS_BoneSpace
    scaleSpace=BCS_BoneSpace
    translationMode=BMM_Additive
    translationSpace=BCS_BoneSpace
    alphaBoolBlend {blendOption=Linear}
    boneToModify {boneName=dyn_collar_bk_01}
    rotation {pitch=90}
    translation {z=5}
   }
   n54 = "Transform (Modify) Bone - Bone: dyn_collar_back_r_01"!(ComponentPose=@n53, Alpha=@n52, Translation="2,0,0", Rotation="0,0,0", Scale:Vector="1,1,1") -> (Pose) {
    rotationMode=BMM_Additive
    rotationSpace=BCS_BoneSpace
    scaleSpace=BCS_BoneSpace
    translationMode=BMM_Additive
    translationSpace=BCS_BoneSpace
    alphaBoolBlend {blendOption=Linear}
    boneToModify {boneName=dyn_collar_back_r_01}
    translation {x=2}
   }
   n55 = "Transform (Modify) Bone - Bone: dyn_collar_back_l_01"!(ComponentPose=@n54, Alpha=@n52, Translation="2,0,0", Rotation="0,0,0", Scale:Vector="1,1,1") -> (Pose) {
    rotationMode=BMM_Additive
    rotationSpace=BCS_BoneSpace
    scaleSpace=BCS_BoneSpace
    translationMode=BMM_Additive
    translationSpace=BCS_BoneSpace
    alphaBoolBlend {blendOption=Linear}
    boneToModify {boneName=dyn_collar_back_l_01}
    translation {x=2}
   }
   Save cached pose 'SavedPose'(Pose=Component To Local(ComponentPose=@n55)) {cacheName=SavedPose}
   n56 = Get LeftMovementAlphaSkydiving()
   n57 = "Transform (Modify) Bone - Bone: deform_side_l"(
    ComponentPose=Local To Component(LocalPose=Use cached pose 'SavedPose'())
    Alpha=@n56
    Translation="-2,0,-20"
    Rotation="10,0,0"
    Scale="1,1,1"
   ) {
    rotationMode=BMM_Additive
    rotationSpace=BCS_BoneSpace
    translationMode=BMM_Additive
    translationSpace=BCS_BoneSpace
    alphaBoolBlend {blendOption=Linear}
    alphaScaleBiasClamp {interpSpeedDecreasing=0.5, interpSpeedIncreasing=0.5}
    boneToModify {boneName=deform_side_l}
    rotation {pitch=10}
    translation {x=-2, z=-20}
   }
   n58 = "Transform (Modify) Bone - Bone: dyn_collar_side_r_01"!(ComponentPose=@n57, Alpha=@n56, Translation="0,0,0", Rotation="-30,0,0", Scale:Vector="1,1,1") -> (Pose) {
    rotationMode=BMM_Additive
    rotationSpace=BCS_BoneSpace
    translationMode=BMM_Additive
    translationSpace=BCS_BoneSpace
    alphaBoolBlend {blendOption=Linear}
    alphaScaleBiasClamp {interpSpeedDecreasing=0.5, interpSpeedIncreasing=0.5}
    boneToModify {boneName=dyn_collar_side_r_01}
    rotation {pitch=-30}
   }
   n59 = "Transform (Modify) Bone - Bone: dyn_collar_side_l_01"!(ComponentPose=@n58, Alpha=@n56, Translation="-2,0,0", Rotation="30,0,0", Scale:Vector="1,1,1") -> (Pose) {
    rotationMode=BMM_Additive
    rotationSpace=BCS_BoneSpace
    translationMode=BMM_Additive
    translationSpace=BCS_BoneSpace
    alphaBoolBlend {blendOption=Linear}
    alphaScaleBiasClamp {interpSpeedDecreasing=0.5, interpSpeedIncreasing=0.5}
    boneToModify {boneName=dyn_collar_side_l_01}
    rotation {pitch=30}
    translation {x=-2}
   }
   n60 = "Transform (Modify) Bone - Bone: dyn_collar_back_l_01"(ComponentPose=@n59, Alpha=@n56, Translation="0,0,0", Rotation="10,0,0", Scale="1,1,1") {
    rotationMode=BMM_Additive
    rotationSpace=BCS_BoneSpace
    translationMode=BMM_Additive
    translationSpace=BCS_BoneSpace
    alphaBoolBlend {blendOption=Linear}
    alphaScaleBiasClamp {interpSpeedDecreasing=0.5, interpSpeedIncreasing=0.5}
    boneToModify {boneName=dyn_collar_back_l_01}
    rotation {pitch=10}
   }
   n61 = "Transform (Modify) Bone - Bone: dyn_collar_back_r_01"(ComponentPose=@n60, Alpha=@n56, Translation="0,0,0", Rotation="-30,0,0", Scale="1,1,1") {
    rotationMode=BMM_Additive
    rotationSpace=BCS_BoneSpace
    translationMode=BMM_Additive
    translationSpace=BCS_BoneSpace
    alphaBoolBlend {blendOption=Linear}
    alphaScaleBiasClamp {interpSpeedDecreasing=0.5, interpSpeedIncreasing=0.5}
    boneToModify {boneName=dyn_collar_back_r_01}
    rotation {pitch=-30}
   }
   n62 = Get RightMovementAlphaSkydiving()
   n63 = "Transform (Modify) Bone - Bone: deform_side_r"(ComponentPose=@n61, Alpha=@n62, Translation="-2,0,-8", Rotation="10,0,0", Scale="1,1,1") {
    rotationMode=BMM_Additive
    rotationSpace=BCS_BoneSpace
    translationMode=BMM_Additive
    translationSpace=BCS_BoneSpace
    alphaBoolBlend {blendOption=Linear}
    alphaScaleBiasClamp {interpSpeedDecreasing=0.5, interpSpeedIncreasing=0.5}
    boneToModify {boneName=deform_side_r}
    rotation {pitch=10}
    translation {x=-2, z=-8}
   }
   n64 = "Transform (Modify) Bone - Bone: dyn_collar_side_r_01"(ComponentPose=@n63, Alpha=@n62, Translation="-2,0,0", Rotation="20,0,0", Scale="1,1,1") {
    rotationMode=BMM_Additive
    rotationSpace=BCS_BoneSpace
    translationMode=BMM_Additive
    translationSpace=BCS_BoneSpace
    alphaBoolBlend {blendOption=Linear}
    alphaScaleBiasClamp {interpSpeedDecreasing=0.5, interpSpeedIncreasing=0.5}
    boneToModify {boneName=dyn_collar_side_r_01}
    rotation {pitch=20}
    translation {x=-2}
   }
   n65 = "Transform (Modify) Bone - Bone: dyn_collar_side_l_01"(ComponentPose=@n64, Alpha=@n62, Translation="0,0,0", Rotation="-30,0,0", Scale="1,1,1") {
    rotationMode=BMM_Additive
    rotationSpace=BCS_BoneSpace
    translationMode=BMM_Additive
    translationSpace=BCS_BoneSpace
    alphaBoolBlend {blendOption=Linear}
    alphaScaleBiasClamp {interpSpeedDecreasing=0.5, interpSpeedIncreasing=0.5}
    boneToModify {boneName=dyn_collar_side_l_01}
    rotation {pitch=-30}
   }
   n66 = "Transform (Modify) Bone - Bone: dyn_collar_back_l_01"(ComponentPose=@n65, Alpha=@n62, Translation="0,0,0", Rotation="-30,0,0", Scale="1,1,1") {
    rotationMode=BMM_Additive
    rotationSpace=BCS_BoneSpace
    translationMode=BMM_Additive
    translationSpace=BCS_BoneSpace
    alphaBoolBlend {blendOption=Linear}
    alphaScaleBiasClamp {interpSpeedDecreasing=0.5, interpSpeedIncreasing=0.5}
    boneToModify {boneName=dyn_collar_back_l_01}
    rotation {pitch=-30}
   }
   n67 = "Transform (Modify) Bone - Bone: dyn_collar_back_r_01"(ComponentPose=@n66, Alpha=@n62, Translation="-2,0,0", Rotation="20,0,0", Scale="1,1,1") {
    rotationMode=BMM_Additive
    rotationSpace=BCS_BoneSpace
    translationMode=BMM_Additive
    translationSpace=BCS_BoneSpace
    alphaBoolBlend {blendOption=Linear}
    alphaScaleBiasClamp {interpSpeedDecreasing=0.5, interpSpeedIncreasing=0.5}
    boneToModify {boneName=dyn_collar_back_r_01}
    rotation {pitch=20}
    translation {x=-2}
   }
   n68 = Get Left arm skyfloating() -> (Left arm skyfloating:double)
   n69 = "Transform (Modify) Bone - Bone: dyn_collar_side_l_01"(ComponentPose=@n67, Alpha=@n68, Translation="-2,0,0", Rotation="5,0,0", Scale="1,1,1") {
    rotationMode=BMM_Additive
    rotationSpace=BCS_BoneSpace
    translationMode=BMM_Additive
    translationSpace=BCS_BoneSpace
    alphaBoolBlend {blendOption=Linear}
    alphaScaleBiasClamp {interpSpeedDecreasing=0.5, interpSpeedIncreasing=0.5}
    boneToModify {boneName=dyn_collar_side_l_01}
    rotation {pitch=5}
    translation {x=-2}
   }
   n70 = "Transform (Modify) Bone - Bone: dyn_collar_back_l_01"(ComponentPose=@n69, Alpha=@n68, Translation="-2,0,0", Rotation="5,0,0", Scale="1,1,1") {
    rotationMode=BMM_Additive
    rotationSpace=BCS_BoneSpace
    translationMode=BMM_Additive
    translationSpace=BCS_BoneSpace
    alphaBoolBlend {blendOption=Linear}
    alphaScaleBiasClamp {interpSpeedDecreasing=0.5, interpSpeedIncreasing=0.5}
    boneToModify {boneName=dyn_collar_back_l_01}
    rotation {pitch=5}
    translation {x=-2}
   }
   n71 = Get Right arm skyfloating() -> (Right arm skyfloating:double)
   n72 = "Transform (Modify) Bone - Bone: dyn_collar_side_r_01"(ComponentPose=@n70, Alpha=@n71, Translation="-2,0,0", Rotation="5,0,0", Scale="1,1,1") {
    rotationMode=BMM_Additive
    rotationSpace=BCS_BoneSpace
    translationMode=BMM_Additive
    translationSpace=BCS_BoneSpace
    alphaBoolBlend {blendOption=Linear}
    alphaScaleBiasClamp {interpSpeedDecreasing=0.5, interpSpeedIncreasing=0.5}
    boneToModify {boneName=dyn_collar_side_r_01}
    rotation {pitch=5}
    translation {x=-2}
   }
   n73 = "Transform (Modify) Bone - Bone: dyn_collar_back_r_01"(ComponentPose=@n72, Alpha=@n71, Translation="-2,0,0", Rotation="5,0,0", Scale="1,1,1") {
    rotationMode=BMM_Additive
    rotationSpace=BCS_BoneSpace
    translationMode=BMM_Additive
    translationSpace=BCS_BoneSpace
    alphaBoolBlend {blendOption=Linear}
    alphaScaleBiasClamp {interpSpeedDecreasing=0.5, interpSpeedIncreasing=0.5}
    boneToModify {boneName=dyn_collar_back_r_01}
    rotation {pitch=5}
    translation {x=-2}
   }
   Save cached pose 'Skydiving Pose'(Pose=Blend Poses by bool(
     BlendPose_0=Component To Local(ComponentPose=@n73)
     BlendPose_1=Use cached pose 'SavedPose'()
     BlendTime_0=0
     BlendTime_1=0
     bActiveValue=Get Skydiving() -> (Skydiving:bool)
    ) {blendType=Linear}) {cacheName=Skydiving Pose}
  ]}
 EventGraph {nodes [
   Event BlueprintInitializeAnimation()
   > Set MeshToCopy(MeshToCopy:Skeletal Mesh Component Object Reference=GetOwnerSkeletalMeshComponent()) -> (Output_Get:Skeletal Mesh Component Object Reference)
   > Set GetPlayerListUsesCustomCharacterParts(GetPlayerListUsesCustomCharacterParts:bool=GetPlaylistUsesCustomCharacterParts!() -> (ReturnValue:bool)) -> (Output_Get:bool)
   Event BlueprintUpdateAnimation()
   > n1 = Sequence!(execute) -> (then_0, then_1, then_2:Exec, then_3:Exec, then_4:Exec, then_5:Exec, then_6:Exec, then_7:Exec, then_8:Exec)
   n2 = GetDeltaTransformFromRefPose(self=GetOwnerSkeletalMeshComponent(), BoneName=head, BaseName=spine_05)
   > Set headSpineRot(headSpineRot:double=BreakRotator(InRot=BreakTransform(InTransform=@n2).Rotation).Yaw) -> (Output_Get:double)
   n3 = Branch(execute=@n1.then_1, Condition=OR Boolean!(A=Get bIsTargeting(), B=Get bIsSkyDiving(), C:bool=Get bIsSkydiveFloating()) -> (ReturnValue))
   > Set head_rot_back_not_targeting(head_rot_back_not_targeting:double=0) -> (Output_Get:double)
   Set head_rot_back_not_targeting(execute=@n3.else, head_rot_back_not_targeting=MapRangeClamped(Value=Get headSpineRot() -> (headSpineRot:double), InRangeA=0, InRangeB=-60))
   n4 = Branch(execute=@n1.then_5, Condition=OR Boolean!(A=Get bIsSkyDiving(), B=Get bIsSkydiveFloating()) -> (ReturnValue))
   > Set head_rot_back(head_rot_back:double=0) -> (Output_Get:double)
   Set head_rot_back(execute=@n4.else, head_rot_back=MapRangeClamped(Value=Get headSpineRot(), InRangeA=0, InRangeB=-60))
   n5 = Branch(execute=@n1.then_6, Condition=OR Boolean(A=Get bIsSkyDiving(), B=Get bIsSkydiveFloating()))
   > Set Neck Side Right(Neck Side Right:float=0) -> (Output_Get:float)
   n6 = Get MeshToCopy()
   n7 = Set Neck Side Right(
    execute=@n5.else
    Neck Side Right=GetAlphaFromReferenceTransformDelta(SkeletalMeshComponent=@n6, BoneName=head, BaseName=spine_05, DeltaComponentType=EulerX, InputRange=(X=0,Y=-28), TwistAxis="1,0,0")
   )
   Set Neck Side Left(execute=[@n8.then, @n7.then], Neck Side Left:float=0) -> (Output_Get:float)
   n8 = Branch(execute=@n1.then_7, Condition=OR Boolean(A=Get bIsSkyDiving(), B=Get bIsSkydiveFloating()))
   Set Neck Side Left(
    execute=@n8.else
    Neck Side Left=GetAlphaFromReferenceTransformDelta(SkeletalMeshComponent=@n6, BoneName=head, BaseName=spine_05, DeltaComponentType=EulerX, InputRange=(X=5,Y=30), TwistAxis="1,0,0")
   )
   n9 = Branch(execute=@n1.then_2, Condition=OR Boolean(A=Get bIsSkyDiving(), B=Get bIsSkydiveFloating()))
   > Set ads_rotate_back(ads_rotate_back:double=0) -> (Output_Get:double)
   Set ads_rotate_back(execute=@n9.else, ads_rotate_back=MapRangeClamped(Value=Get headSpineRot(), InRangeA=0, InRangeB=-50, OutRangeA=0.5))
   Set Head_Rot_Forward(execute=@n1.then_3, Head_Rot_Forward:double=MapRangeClamped(Value=Get headSpineRot(), InRangeA=0, InRangeB=30)) -> (Output_Get:double)
   Set ADS_Rot_Forward(execute=@n1.then_4, ADS_Rot_Forward:double=MapRangeClamped(Value=Get headSpineRot(), InRangeA=0, InRangeB=30)) -> (Output_Get:double)
   n10 = Get MeshToCopy()
   > Set Head Pitch(Head Pitch:float=GetAlphaFromReferenceTransformDelta(SkeletalMeshComponent=@n10, BoneName=head, BaseName=spine_05, InputRange=(X=0,Y=28), TwistAxis="1,0,0")) -> (Output_Get:float)
   > Set Neck 01 Rotation L(Neck 01 Rotation L:double=GetAlphaFromReferenceTransformDelta(SkeletalMeshComponent=@n10, BoneName=neck_01, BaseName=spine_05, DeltaComponentType=EulerX, InputRange=(X=0,Y=-28), TwistAxis="1,0,0")) -> (Output_Get:double)
   > Set Neck 01 Rotation R(Neck 01 Rotation R:float=GetAlphaFromReferenceTransformDelta(SkeletalMeshComponent=@n10, BoneName=head, BaseName=spine_05, DeltaComponentType=EulerX, InputRange=(X=-5,Y=-30), TwistAxis="1,0,0")) -> (Output_Get:float)
   > Set Head Yaw(Head Yaw:float=GetAlphaFromReferenceTransformDelta(SkeletalMeshComponent=@n10, BoneName=head, BaseName=spine_05, DeltaComponentType=EulerY, InputRange=(X=-5,Y=15), TwistAxis="1,0,0")) -> (Output_Get:float)
   n11 = TryGetPawnOwner()
   n12 = Dot Product!(
    A:Vector=GetVelocity!(self:Actor Object Reference=@n11) -> (ReturnValue:Vector)
    B:Vector=GetActorRightVector!(self:Actor Object Reference=@n11) -> (ReturnValue:Vector)
   ) -> (ReturnValue:double)
   n13 = Get bIsTargeting()
   n14 = Get bIsSkyDiving()
   n15 = Get bIsSkydiveFloating()
   > Set RightMovementAlphaSkydiving(RightMovementAlphaSkydiving:double=SelectFloat(A=MapRangeClamped(Value=@n12, InRangeA=0, InRangeB=230), B=0, bPickA=OR Boolean!(A=@n13, B=@n14, C:bool=@n15) -> (ReturnValue))) -> (Output_Get:double)
   n16 = Get bIsTargeting()
   n17 = Get bIsSkydiveFloating()
   n18 = Get bIsSkyDiving()
   > Set LeftMovementAlphaSkydiving(LeftMovementAlphaSkydiving:double=SelectFloat(A=MapRangeClamped(Value=@n12, InRangeA=0, InRangeB=-230), B=0, bPickA=OR Boolean(A=@n16, B=@n18, C=@n17))) -> (Output_Get:double)
   n19 = Get bIsSkydiveFloating()
   > Set Skydiving(Skydiving:bool=AND Boolean(A=Get bIsSkyDiving(), B=NOT Boolean(A=@n19))) -> (Output_Get:bool)
   n20 = Get headSpineRot()
   n21 = Get bIsMoving2D()
   n22 = Get bIsTargeting()
   n23 = Get bIsSkydiveFloating()
   n24 = Get bIsSkyDiving()
   > "Set Idle "("Idle ":double=SelectFloat(A=0, B=MapRangeClamped(Value=@n20, InRangeA=0, InRangeB=-60), bPickA=OR Boolean!(A=@n22, B=@n24, C=@n23, D:bool=@n21) -> (ReturnValue))) -> (Output_Get:double)
   n25 = Get MeshToCopy()
   n26 = Get bIsTargeting()
   n27 = Get bIsSkyDiving()
   n28 = Get bIsSkydiveFloating()
   > Set Alpha Left Arm(Alpha Left Arm:float=SelectFloat(
     A=0
     B=GetAlphaFromReferenceTransformDelta(SkeletalMeshComponent=@n25, BoneName=upperarm_l, BaseName=clavicle_l, InputRange=(X=0,Y=75), TwistAxis="1,0,0")
     bPickA=OR Boolean!(A=@n26, B=@n27, C=@n28) -> (ReturnValue)
    )) -> (Output_Get:float)
   n29 = Get MeshToCopy()
   n30 = Get bIsSkyDiving()
   n31 = Get bIsSkydiveFloating()
   > Set Alpha Output Idle(Alpha Output Idle:float=SelectFloat(
     A=0
     B=GetAlphaFromReferenceTransformDelta(SkeletalMeshComponent=@n29, BoneName=head, BaseName=spine_05, InputRange=(X=-40,Y=-70), TwistAxis="1,0,0")
     bPickA=OR Boolean!(A=@n30, B=@n31) -> (ReturnValue)
    )) -> (Output_Get:float)
   n32 = NOT Boolean(A=Get bIsSkyDiving())
   n33 = NOT Boolean(A=Get bIsTargeting())
   n34 = NOT Boolean(A=Get bIsSkydiveFloating())
   n35 = OR Boolean(A=Get bIsJumping(), B=Get bIsFalling())
   > n36 = "Set  jump alpha"(" jump alpha":double=SelectFloat(A=1, B=0, bPickA=AND Boolean!(A=@n35, B=@n33, C:bool=@n32, D:bool=@n34) -> (ReturnValue))) -> (Output_Get:double)
   n37 = Get headSpineRot()
   n38 = Get bIsSkyDiving()
   n39 = Get bIsSkydiveFloating()
   Set Head rotation back skydiving(
    execute=[@n55.then, @n36.then]
    Head rotation back skydiving:double=SelectFloat(A=MapRangeClamped(Value=@n37, InRangeA=0, InRangeB=-55), B=0, bPickA=OR Boolean(A=@n38, B=@n39))
   ) -> (Output_Get:double)
   > Set Head rotation back skyfloating(Head rotation back skyfloating:double=MapRangeClamped(Value=Get headSpineRot(), InRangeA=-5, InRangeB=-30)) -> (Output_Get:double)
   n40 = Get MeshToCopy()
   n41 = Get bIsSkyDiving()
   n42 = Get bIsSkydiveFloating()
   > Set Left arm skydiving(Left arm skydiving:double=SelectFloat(
     A=GetAlphaFromReferenceTransformDelta(SkeletalMeshComponent=@n40, BoneName=upperarm_l, BaseName=clavicle_l, InputRange=(X=0,Y=22), TwistAxis="1,0,0")
     B=0
     bPickA=OR Boolean(A=@n41, B=@n42)
    )) -> (Output_Get:double)
   n43 = Get MeshToCopy()
   n44 = Get bIsSkyDiving()
   n45 = Get bIsSkydiveFloating()
   > Set Right arm skydiving(Right arm skydiving:double=SelectFloat(
     A=GetAlphaFromReferenceTransformDelta(SkeletalMeshComponent=@n43, BoneName=upperarm_r, BaseName=clavicle_r, InputRange=(X=0,Y=22), TwistAxis="1,0,0")
     B=0
     bPickA=OR Boolean(A=@n44, B=@n45)
    )) -> (Output_Get:double)
   n46 = Get MeshToCopy()
   n47 = NOT Boolean(A=Get bIsSkydiveFloating())
   n48 = Get bIsSkyDiving()
   > Set skydiving neck rot(skydiving neck rot:float=SelectFloat(
     A=GetAlphaFromReferenceTransformDelta(SkeletalMeshComponent=@n46, BoneName=head, BaseName=spine_05, DeltaComponentType=EulerX, InputRange=(X=-5,Y=5), TwistAxis="1,0,0")
     B=0
     bPickA=OR Boolean(A=@n48, B=@n47)
    )) -> (Output_Get:float)
   n49 = Get MeshToCopy()
   n50 = Get bIsSkyDiving()
   n51 = Get bIsSkydiveFloating()
   > Set Left arm skyfloating(Left arm skyfloating:double=SelectFloat(
     A=GetAlphaFromReferenceTransformDelta(SkeletalMeshComponent=@n49, BoneName=upperarm_l, BaseName=clavicle_l, InputRange=(X=0,Y=18), TwistAxis="1,0,0")
     B=0
     bPickA=OR Boolean(A=@n50, B=@n51)
    )) -> (Output_Get:double)
   n52 = Get MeshToCopy()
   n53 = Get bIsSkyDiving()
   n54 = Get bIsSkydiveFloating()
   > Set Right arm skyfloating(Right arm skyfloating:double=SelectFloat(
     A=GetAlphaFromReferenceTransformDelta(SkeletalMeshComponent=@n52, BoneName=upperarm_r, BaseName=clavicle_r, InputRange=(X=0,Y=18), TwistAxis="1,0,0")
     B=0
     bPickA=OR Boolean(A=@n53, B=@n54)
    )) -> (Output_Get:double)
   n55 = Set Head rotation back skydiving idle(Head rotation back skydiving idle:double=MapRangeClamped(Value=Get headSpineRot(), InRangeA=0, InRangeB=-55)) -> (Output_Get:double)
   TryGetPawnOwner()
  ]}
]
variables [
 " jump alpha" {type=number}
 ADS_Rot_Forward {type=number}
 ads_rotate_back {type=number}
 Alpha Left Arm {type=number}
 Alpha Output Idle {type=number}
 GetPlayerListUsesCustomCharacterParts {type=boolean}
 Head Pitch {type=number}
 Head rotation back skydiving {type=number}
 Head rotation back skydiving idle {type=number}
 Head rotation back skyfloating {type=number}
 Head Yaw {type=number}
 head_rot_back {type=number}
 head_rot_back_not_targeting {type=number}
 Head_Rot_Forward {type=number}
 headSpineRot {type=number}
 "Idle " {type=number}
 Left arm skydiving {type=number}
 Left arm skyfloating {type=number}
 Left mvmnt corr {type=boolean}
 LeftMovementAlphaSkydiving {type=number}
 MeshToCopy {type=SkeletalMeshComponent}
 Neck 01 Rotation L {type=number}
 Neck 01 Rotation R {type=number}
 Neck Side Left {type=number}
 Neck Side Right {type=number}
 Right arm skydiving {type=number}
 Right arm skyfloating {type=number}
 Right mvmnt corr {type=boolean}
 RightMovementAlphaSkydiving {type=number}
 Skydiving {type=boolean}
 skydiving neck rot {type=number}
 Start moving left {type=number}
 Start moving right {type=number}
]
